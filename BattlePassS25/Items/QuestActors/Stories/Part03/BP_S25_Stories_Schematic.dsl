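; Battle Pass S25 story quest prop (Schematic) that turns a nearby Stingray NPC hostile when a player completes its quest interaction.
; On quest component success, the server waits briefly, passes a DoOnce gate, stores the player's pawn as instigator and calls the aggro event.
; That event searches actors overlapping an encounter bounding box for a pawn with the Stingray spawner tag.
; If found, it sets the NPC's attitude toward the player to Hostile and calls Reset DoOnce to re-arm the gate.
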(event EventBeginPlay
  (|CallParentFunction) ; This node is disabled and will not be called.\nDrag off pins to build functionality.
)

(event OnQuestComponentSuccessEvent(ConsolidatedQuestComponent)
  (params (PlayerController :type Fort%20Player%20Controller%20Object%20Reference))
  (bind _returnvalue (Game|Player|GetPlayerPawn PlayerController))
  ;[0 0 0] On successful interaction, turn Stingray in range hostile towards the player, if they exist
  (Networking|SwitchHasAuthority
    (:Authority
      (Utilities|FlowControl|Delay 0.5)
      (Utilities|FlowControl|DoOnce false
        (:Completed
          (Utilities|IsValid _returnvalue
            (:Is%20Valid
              (Variables|Components|SetInstigatingPlayerPawn _returnvalue)
              (CallFunction|AggroStingrayinBounds))))))))

(event Custom|Reset%20DoOnce
  ;[0 0 0] On successful interaction, turn Stingray in range hostile towards the player, if they exist
  (Utilities|FlowControl|DoOnce false
    (:Completed
      (Utilities|IsValid (Game|Player|GetPlayerPawn (stale OnQuestComponentSuccessEvent(ConsolidatedQuestComponent) PlayerController))
        (:Is%20Valid
          (Variables|Components|SetInstigatingPlayerPawn (Game|Player|GetPlayerPawn (stale OnQuestComponentSuccessEvent(ConsolidatedQuestComponent) PlayerController)))
          (CallFunction|AggroStingrayinBounds))))))

(event Collision|EventActorBeginOverlap
  (params (OtherActor :type Actor%20Object%20Reference))
  (|CallParentFunction OtherActor) ; This node is disabled and will not be called.\nDrag off pins to build functionality.
)

(event EventTick
  (params (DeltaSeconds :type Float%20%28single-precision%29))
  (|CallParentFunction DeltaSeconds) ; This node is disabled and will not be called.\nDrag off pins to build functionality.
)

(event Custom|Aggro%20Stingray%20In%20Bounds
  (bind _stingraynpc (Variables|Components|GetStingrayNPC))
  ;[0 0 0] Find the Stingray and aggro him
  (bind _array_element (Utilities|Array|ForEachLoopwithBreak (Collision|GetOverlappingActors (Variables|Default|GetBoundingBoxEncounter) (Variables|Components|GetTandemClassFilter) :!class /Script/Engine.PrimitiveComponent))
    (:LoopBody
      (bind _asfort_player_pawn (Utilities|Casting|CastToFortPlayerPawn _array_element))
      (Utilities|IsValid _asfort_player_pawn
        (:Is%20Valid
          (if (GameplayTags|HasTag (GameplayTags|GetOwnedGameplayTags _asfort_player_pawn :!class /Script/GameplayTags.GameplayTagAssetInterface) (Variables|Components|GetStingraySpawnerDataTag) true)
            (Variables|Components|SetStingrayNPC _asfort_player_pawn)))))
    (:Completed
      _))
  ;;[0.224 0.224 0.224] If henchman was found, turn it hostile and reset the DoOnce Loop in Quest Success for this prop.
  (Utilities|IsValid _stingraynpc
    (:Is%20Valid
      (Affiliation|ChangeAttitudeBetweenActors _stingraynpc (Variables|Components|GetInstigatingPlayerPawn) "Hostile")
      (CallFunction|ResetDoOnce))))

(fn ConstructionScript
  (|CallParentFunction))
